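# Sound cue for the nighttime metal room ambience at the Hollow Base Takeover POI, routed to the indoor Athena ambient sound class.
# A mixer layers two looping waves: an industrial room tone at reduced level and a boosted crypt interior morning loop reused from Starlight.
# It plays without spatialization, attenuation or reverb send, and its concurrency override prevents new instances once the limit is reached.
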SoundCue /Game/Sounds/Artemis/POI/Hollow_Base_Takover/Ambience_Hollow_Base_Takeover_Metal_Room_Night_Cue
root=Mixer_0
nodes [
 {id=Mixer_0, children [WavePlayer_3,WavePlayer_2], properties {inputVolume [0.5,2]}}
 {
  id=WavePlayer_2
  properties {bLooping=true, soundWaveAssetPtr=/Game/Sounds/Fort_Starlight/S1/SoundWaves/AMBI_Starlight_mToD_Crypt_Int_Morning_02_Loop}
 }
 {id=WavePlayer_3, properties {bLooping=true, soundWaveAssetPtr=/Game/Sounds/Trans_AmbientLoops/RoomTones/IndustrialSmall002_Loop_Stereo}}
]
properties {
 bOverrideConcurrency=true
 maxDistance=2097152
 soundClassObject=&Fort_Ambient_Indoors_Athena
 volumeMultiplier=0.652
 attenuationOverrides {bAttenuate=false, bEnableReverbSend=false, bSpatialize=false}
 concurrencyOverrides {resolutionRule=PreventNew}
}
